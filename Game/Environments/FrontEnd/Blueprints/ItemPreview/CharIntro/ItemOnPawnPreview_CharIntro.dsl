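; Item-on-pawn preview Blueprint for the front-end character intro.
; When its preview visuals are shown, it runs the parent behavior, then activates the CBM_CharIntro audio modulation control bus mix.
; The construction script only calls the parent, and the file shows no matching deactivation of the mix.
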(event EventOnPreviewVisualsShown
  (|CallParentFunction)
  (Audio|Modulation|ActivateControlBusMix "/Game/Sounds/Modulation/Mixes/CBM_CharIntro.CBM_CharIntro"))

(fn ConstructionScript
  (|CallParentFunction))
